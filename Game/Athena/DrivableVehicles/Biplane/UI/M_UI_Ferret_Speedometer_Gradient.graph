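# Translucent UI material that, by its name, draws the gradient for the biplane (Ferret) speedometer.
# An exponential radial gradient on UV channel 1 passes through SmoothStep, starting at GaugeEdge, to form a circular mask.
# The transition width is GaugeEdge times EdgeSoftness, so softness scales with the edge threshold.
# A vertical linear gradient on the same UV channel is subtracted from that mask to fade it along V.
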Material /Game/Athena/DrivableVehicles/Biplane/UI/M_UI_Ferret_Speedometer_Gradient
graph {
 properties {blendMode=BLEND_Translucent, materialDomain=MD_UI, pixelDepthOffsetMode=PDOM_Legacy, usedShadingModels=Default Lit}
 nodes [
  n1 = ScalarParameter() {defaultValue=0.02, desc=0.02, group=Gauge, parameterName=GaugeEdge}
  n2 = Multiply(A=@n1, B=ScalarParameter() {defaultValue=0.3, desc=0.2, parameterName=EdgeSoftness}) {constB=1.01}
  n3 = TextureCoordinate() {coordinateIndex=1}
  Subtract(
   A=MaterialFunctionCall(
    Alpha=MaterialFunctionCall(UVs=@n3) {materialFunction=&RadialGradientExponential}.RadialGradientExponential
    Min=@n1
    Max=Add(A=@n1, B=@n2)
   ) {materialFunction=&SmoothStep}.Result
   B=MaterialFunctionCall(UV Channel=TextureCoordinate() {coordinateIndex=1}) {materialFunction=&LinearGradient}.VGradient
  )
 ]
}
