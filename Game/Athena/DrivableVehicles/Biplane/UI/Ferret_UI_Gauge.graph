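# Biplane (Ferret) dashboard widget showing radial fuel, boost and speed gauges, plus fuel and disabled-vehicle warnings.
# Construct caches dynamic materials; parent-class events drive their scalar parameters, with boost glow changing at full charge.
# Fuel events toggle the meter for infinite fuel, switch low/no-fuel labels and icons (material names appear swapped), and play warning animations.
# Async gameplay-tag waits on Vehicle.Disabled toggle the shared vehicle warning widget.
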WidgetBlueprint /Game/Athena/DrivableVehicles/Biplane/UI/Ferret_UI_Gauge
animations [
 IsLowFuel {
  binding_count=3
  length_seconds=0.6074
  object=&self:IsLowFuel
  sequence {bindings [
    Overlay_LowFuel {
     id=#1
     possessed_class=Overlay
     tracks [{
       type=MovieScene2DTransformTrack
       sections [{
         type=MovieScene2DTransformSection
         channels [
          Translation.X {type=MovieSceneScriptingFloatChannel}
          Translation.Y {type=MovieSceneScriptingFloatChannel}
          Angle {type=MovieSceneScriptingFloatChannel}
          Scale.X {type=MovieSceneScriptingFloatChannel, keys [[0,0],[5.1863,0.9605],[5.4153,1]]}
          Scale.Y {type=MovieSceneScriptingFloatChannel, keys [[0,0],[5.1983,0.9654],[5.4153,1]]}
          Shear.X {type=MovieSceneScriptingFloatChannel}
          Shear.Y {type=MovieSceneScriptingFloatChannel}
         ]
        }]
      }]
    }
    "Overlay_LowFuel (2)" {
     id=#2
     display_name=Image_RoundCard_Bolt
     possessed_class=Image
     tracks [{
       type=MovieScene2DTransformTrack
       sections [{
         type=MovieScene2DTransformSection
         channels [
          Translation.X {type=MovieSceneScriptingFloatChannel}
          Translation.Y {type=MovieSceneScriptingFloatChannel}
          Angle {type=MovieSceneScriptingFloatChannel}
          Scale.X {type=MovieSceneScriptingFloatChannel, keys [[4.2267,0],[8.2807,1]]}
          Scale.Y {type=MovieSceneScriptingFloatChannel, keys [[4.2267,0],[8.2807,1]]}
          Shear.X {type=MovieSceneScriptingFloatChannel}
          Shear.Y {type=MovieSceneScriptingFloatChannel}
         ]
        }]
      }]
    }
    CommonTextBlock_57 {
     id=#3
     possessed_class=CommonTextBlock
     tracks [
      {
       type=MovieScene2DTransformTrack
       sections [{
         type=MovieScene2DTransformSection
         channels [
          Translation.X {type=MovieSceneScriptingFloatChannel, keys [[6.7033,-231.2176],[12.148,0]]}
          Translation.Y {type=MovieSceneScriptingFloatChannel}
          Angle {type=MovieSceneScriptingFloatChannel}
          Scale.X {type=MovieSceneScriptingFloatChannel}
          Scale.Y {type=MovieSceneScriptingFloatChannel}
          Shear.X {type=MovieSceneScriptingFloatChannel}
          Shear.Y {type=MovieSceneScriptingFloatChannel}
         ]
        }]
      }
      {
       type=MovieSceneFloatTrack
       display_name=Render Opacity
       sections [{type=MovieSceneFloatSection, channels [None {type=MovieSceneScriptingFloatChannel, keys [[6.7033,0],[9.158,1]]}]}]
      }
     ]
    }
   ]}
 }
 NoFuelMessage {
  binding_count=2
  length_seconds=2.7194
  object=&self:NoFuelMessage
  sequence {bindings [
    Overlay_NoFuelMessage {
     id=#4
     possessed_class=Overlay
     tracks [
      {
       type=MovieSceneByteTrack
       display_name=Visibility
       sections [{type=MovieSceneByteSection, channels [None {type=MovieSceneScriptingByteChannel, keys [[0,1],[1,3],[54.388,1]]}]}]
      }
      {
       type=MovieSceneFloatTrack
       display_name=Render Opacity
       sections [{type=MovieSceneFloatSection, channels [None {type=MovieSceneScriptingFloatChannel, keys [[1,0],[3,1],[40.662,1],[54.388,0]]}]}]
      }
     ]
    }
    CommonTextBlock_15 {
     id=#5
     possessed_class=CommonTextBlock
     tracks [{
       type=MovieScene2DTransformTrack
       sections [{
         type=MovieScene2DTransformSection
         channels [
          Translation.X {type=MovieSceneScriptingFloatChannel}
          Translation.Y {type=MovieSceneScriptingFloatChannel}
          Angle {type=MovieSceneScriptingFloatChannel}
          Scale.X {type=MovieSceneScriptingFloatChannel, keys [[3,1],[4.2057,2.7122],[6,2.9755],[8,1],[8.8983,1]]}
          Scale.Y {type=MovieSceneScriptingFloatChannel, keys [[3,1],[4.2057,2.7122],[6,2.9755],[8,1],[8.8983,1]]}
          Shear.X {type=MovieSceneScriptingFloatChannel}
          Shear.Y {type=MovieSceneScriptingFloatChannel}
         ]
        }]
      }]
    }
   ]}
 }
]
graphs [
 EventGraph {nodes [
   n1 = Event OnWidgetVisibilityChanged!() -> (OutputDelegate, then, bNewVisibility:bool) {is=title K2Node_Event}
   > n2 = Branch(Condition=@n1.bNewVisibility)
   StopAnimation(execute=@n2.else, InAnimation=Get NoFuelMessage() -> (NoFuelMessage:Widget Animation Object Reference))
   n3 = Event SetVehicleHasInfiniteFuel!() -> (OutputDelegate, then, bInfiniteFuel:bool) {is=title K2Node_Event}
   > SetVisibility(
    self=Get Overlay_FuelMeter() -> (Overlay_FuelMeter:Overlay Object Reference)
    InVisibility=Select(Option 0=HitTestInvisible, Option 1=Collapsed, Index=@n3.bInfiniteFuel)
   )
   Event ShowNoFuelMessage() {is=title K2Node_Event}
   > PlayAnimationForward(InAnimation=Get NoFuelMessage())
   Event OnOutOfFuel() {is=title K2Node_Event}
   > SetActiveWidgetIndex(self=Get Switcher_FuelWarning() -> (Switcher_FuelWarning:Common Widget Switcher Legacy Object Reference), Index=1)
   > SetBrushFromMaterial(
    self=Get Image_LowFuelIcon() -> (Image_LowFuelIcon:Image Object Reference)
    Material=Get LowFuelMaterialInstance() -> (LowFuelMaterialInstance:Material Interface Object Reference)
   )
   n4 = Event OnVehicleUpdated!() -> (OutputDelegate, then, NewVehicle:Fort Athena Vehicle Object Reference) {is=title K2Node_Event}
   n5 = Get DisabledTag() -> (DisabledTag:Gameplay Tag Structure)
   > n6 = WaitGameplayTagAddToActor!(execute:Exec, TargetActor:Actor Object Reference=@n4.NewVehicle, Tag:Gameplay Tag Structure=@n5, OnlyTriggerOnce:bool=false) -> (then:Exec, AsyncTaskProxy:Ability Async Wait Gameplay Tag Added Object Reference, Added:Exec) {is=title K2Node_AsyncAction}
   > n7 = WaitGameplayTagRemoveFromActor!(execute:Exec, TargetActor:Actor Object Reference=@n4.NewVehicle, Tag:Gameplay Tag Structure=@n5, OnlyTriggerOnce:bool=false) -> (then:Exec, AsyncTaskProxy:Ability Async Wait Gameplay Tag Removed Object Reference, Removed:Exec) {is=title K2Node_AsyncAction}
   OnVehicleEnabled!(execute=@n7.Removed, self, Enabled:bool=true) -> (then)
   OnVehicleEnabled(execute=@n6.Added, Enabled=false)
   n8 = Event UpdateSpeedometer!() -> (OutputDelegate, then, NormalizedSpeed:float, RawSpeed:float) {is=title K2Node_Event}
   > SetScalarParameterValue(
    self=Get SpeedMeterMaterial() -> (SpeedMeterMaterial:Material Instance Dynamic Object Reference)
    ParameterName=Get GaugeFillParamName() -> (GaugeFillParamName:Name=None)
    Value=@n8.NormalizedSpeed
   )
   Event OnLowFuel() {is=title K2Node_Event}
   > SetActiveWidgetIndex(self=Get Switcher_FuelWarning(), Index=0)
   > SetBrushFromMaterial(self=Get Image_LowFuelIcon(), Material=Get FuelEmptyMaterialInstance() -> (FuelEmptyMaterialInstance:Material Interface Object Reference))
   > PlayAnimation(InAnimation=Get IsLowFuel() -> (IsLowFuel:Widget Animation Object Reference), bRestoreState=true)
   Event Construct()
   > n9 = GetDynamicMaterial(self=Get FuelMeter() -> (FuelMeter:Image Object Reference))
   > Set FuelMeterMaterial(FuelMeterMaterial:Material Instance Dynamic Object Reference=@n9) -> (Output_Get:Material Instance Dynamic Object Reference)
   > n10 = GetDynamicMaterial(self=Get BoostMeter() -> (BoostMeter:Image Object Reference))
   > Set BoostMeterMaterial(BoostMeterMaterial:Material Instance Dynamic Object Reference=@n10) -> (Output_Get:Material Instance Dynamic Object Reference)
   > n11 = GetDynamicMaterial(self=Get SpeedMeter() -> (SpeedMeter:Image Object Reference))
   > Set SpeedMeterMaterial(SpeedMeterMaterial:Material Instance Dynamic Object Reference=@n11) -> (Output_Get:Material Instance Dynamic Object Reference)
   > n12 = GetDynamicMaterial(self=Get BoostGlowMeter() -> (BoostGlowMeter:Image Object Reference))
   > Set BoostGlowMeterMaterial(BoostGlowMeterMaterial:Material Instance Dynamic Object Reference=@n12) -> (Output_Get:Material Instance Dynamic Object Reference)
   n13 = Event OnFuelValueChanged!() -> (OutputDelegate, then, FuelPercentage:float, bIsLowOnFuel:bool) {is=title K2Node_Event}
   > SetScalarParameterValue(
    self=Get FuelMeterMaterial() -> (FuelMeterMaterial:Material Instance Dynamic Object Reference)
    ParameterName=Get GaugeFillParamName() -> (GaugeFillParamName=None)
    Value=@n13.FuelPercentage
   )
   > SetVisibility(
    self=Get Overlay_LowFuel() -> (Overlay_LowFuel:Overlay Object Reference)
    InVisibility=Select(Option 0=Collapsed, Option 1=SelfHitTestInvisible, Index=@n13.bIsLowOnFuel)
   )
   n14 = Event OnUpdateBoostChargePercent!() -> (OutputDelegate, then, BoostPercent:float) {is=title K2Node_Event}
   > SetScalarParameterValue(
    self=Get BoostMeterMaterial() -> (BoostMeterMaterial:Material Instance Dynamic Object Reference)
    ParameterName=Get GaugeFillParamName() -> (GaugeFillParamName=None)
    Value=@n14.BoostPercent
   )
   > SetScalarParameterValue(
    self=Get BoostGlowMeterMaterial() -> (BoostGlowMeterMaterial:Material Instance Dynamic Object Reference)
    ParameterName=Get GaugeFillGlowParamName() -> (GaugeFillGlowParamName:Name=None)
    Value=SelectFloat(A=1.25, B=0.9, bPickA="float >= float"(A=@n14.BoostPercent, B=1) {is=title K2Node_CallFunction})
   )
  ]}
 OnVehicleEnabled {nodes [
   n1 = OnVehicleEnabled!() -> (then, Enabled:bool=~) {is=title K2Node_FunctionEntry}
   > n2 = Branch(Condition=@n1)
   n3 = Get WBP_Vehicle_Warning() -> (WBP_Vehicle_Warning:WBP Vehicle Warning Object Reference)
   > RemoveWarning!(execute, self:WBP Vehicle Warning Object Reference=@n3, Type:ENUM_WarningType Enum=NewEnumerator2) -> (then)
   > SetVisibility(self=@n3, InVisibility=Collapsed)
   ShowWarning!(execute=@n2.else, self:WBP Vehicle Warning Object Reference=@n3, Type:ENUM_WarningType Enum=NewEnumerator2) -> (then)
   > SetVisibility(self=@n3, InVisibility=HitTestInvisible)
  ]}
]
variables [
 BoostGlowMeterMaterial {type=MaterialInstanceDynamic}
 BoostMeterMaterial {type=MaterialInstanceDynamic}
 DefaultFuelWarningIconRenderOpacity {type=number, default=1}
 DisabledTag {type=GameplayTag, default {tagName=Vehicle.Disabled}}
 FuelEmptyMaterialInstance {type=MaterialInterface, default=/Game/Athena/HUD/Vehicles/Materials/M_UI_LowFuel_Empty}
 FuelMeterFillColor_Default {type=LinearColor, default=<rgba 1 1 1 1>}
 FuelMeterFillColor_LowFuel {type=LinearColor, default=<rgba 1 0.1137 0 1>}
 FuelMeterFillColorParamName {type=string, default=FillLineColor}
 FuelMeterMaterial {type=MaterialInstanceDynamic}
 GaugeFillGlowParamName {type=string, default=FillWidth}
 GaugeFillParamName {type=string, default=GaugeFill}
 LowFuelAnimationTimerHandle {type=TimerHandle}
 LowFuelMaterialInstance {type=MaterialInterface, default=/Game/Athena/HUD/Vehicles/Materials/M_UI_LowFuel_Inst}
 PulseInterval {type=number, default=2}
 SpeedMeterMaterial {type=MaterialInstanceDynamic}
 WarningIconColor_Off {type=LinearColor, default=<rgba 1 1 1 0.5>}
 WarningIconColor_On {type=LinearColor, default=<rgba 1 0.1137 0 1>}
]
widgets {
 info {
  inheritedWidgetCount=10
  parentClass=/Script/FortniteUI.FortAthenaVehicleDashboardWidget_Ferret
  rootWidgetClass=&SafeZone
  widgetCount=50
 }
 widgets [
  SafeZone_Gauge {widgetClassPath=&SafeZone, bIsVariable=true, widget=SafeZone_Gauge}
  Overlay_160 {
   widgetClassPath=&Overlay
   parent=SafeZone_Gauge
   widget=Overlay_160
   properties {visibility=HitTestInvisible}
   slot {class=SafeZoneSlot}
  }
  Overlay_Gauge {
   widgetClassPath=&Overlay
   bInherited=true
   parent=Overlay_160
   widget=Overlay_Gauge
   slot {class=OverlaySlot, properties {horizontalAlignment=HAlign_Right, verticalAlignment=VAlign_Bottom}}
  }
  VerticalBox_0 {
   widgetClassPath=&VerticalBox
   parent=Overlay_Gauge
   widget=VerticalBox_0
   slot {class=OverlaySlot, properties {horizontalAlignment=HAlign_Right, verticalAlignment=VAlign_Bottom, padding=<ltrb 0 0 60 51>}}
  }
  Overlay_FuelContainer {
   widgetClassPath=&Overlay
   bIsVariable=true
   parent=VerticalBox_0
   widget=Overlay_FuelContainer
   slot {class=VerticalBoxSlot, properties {padding=<ltrb 0 0 0 4>}}
  }
  Overlay_FuelMeter {
   widgetClassPath=&Overlay
   bIsVariable=true
   parent=Overlay_FuelContainer
   widget=Overlay_FuelMeter
   slot {class=OverlaySlot, properties {horizontalAlignment=HAlign_Right}}
  }
  BGGradientFuel {
   widgetClassPath=&Image
   bIsVariable=true
   parent=Overlay_FuelMeter
   widget=BGGradientFuel
   properties {
    visibility=SelfHitTestInvisible
    brush {
     bIsDynamicallyLoaded=false
     drawAs=Image
     imageType=NoImage
     mirroring=NoMirror
     resourceName=None
     resourceObject=&/M_UI_Ferret_Speedometer_Gradient
     tiling=NoTile
     imageSize=<163 163>
     margin=<ltrb 0 0 0 0>
     outlineSettings {
      bUseBrushTransparency=false
      roundingType=HalfHeightRadius
      width=0
      color {colorUseRule=UseColor_Specified, specifiedColor=<rgba 0 0 0 0>}
      cornerRadii=<0 0 0 0>
     }
     tintColor {colorUseRule=UseColor_Specified, specifiedColor=<rgba 0 0 0 1>}
     uVRegion {bIsValid=false, max=<0 0>, min=<0 0>}
    }
    colorAndOpacity=<rgba 1 1 1 0.6>
   }
   slot {class=OverlaySlot, properties {horizontalAlignment=HAlign_Center, verticalAlignment=VAlign_Center}}
  }
  FuelThinRing {
   widgetClassPath=&Image
   bIsVariable=true
   parent=Overlay_FuelMeter
   widget=FuelThinRing
   properties {
    visibility=SelfHitTestInvisible
    brush {
     bIsDynamicallyLoaded=false
     drawAs=Image
     imageType=NoImage
     mirroring=NoMirror
     resourceName=None
     resourceObject=&/M_UI_Ferret_Speedometer_Inst_ThinRing
     tiling=NoTile
     imageSize=<185 185>
     margin=<ltrb 0 0 0 0>
     outlineSettings {
      bUseBrushTransparency=false
      roundingType=HalfHeightRadius
      width=0
      color {colorUseRule=UseColor_Specified, specifiedColor=<rgba 0 0 0 0>}
      cornerRadii=<0 0 0 0>
     }
     tintColor {colorUseRule=UseColor_Specified, specifiedColor=<rgba 1 1 1 1>}
     uVRegion {bIsValid=false, max=<0 0>, min=<0 0>}
    }
   }
   slot {class=OverlaySlot}
  }
  FuelGaugeImage {
   widgetClassPath=&Image
   bIsVariable=true
   parent=Overlay_FuelMeter
   widget=FuelGaugeImage
   properties {
    visibility=SelfHitTestInvisible
    brush {
     bIsDynamicallyLoaded=false
     drawAs=Image
     imageType=NoImage
     mirroring=NoMirror
     resourceName=None
     resourceObject=&/M_UI_Ferret_Speedometer_Inst_Fuel
     tiling=NoTile
     imageSize=<185 185>
     margin=<ltrb 0 0 0 0>
     outlineSettings {
      bUseBrushTransparency=false
      roundingType=HalfHeightRadius
      width=0
      color {colorUseRule=UseColor_Specified, specifiedColor=<rgba 0 0 0 0>}
      cornerRadii=<0 0 0 0>
     }
     tintColor {colorUseRule=UseColor_Specified, specifiedColor=<rgba 1 1 1 1>}
     uVRegion {bIsValid=false, max=<0 0>, min=<0 0>}
    }
   }
   slot {class=OverlaySlot}
  }
  FuelMeter {
   widgetClassPath=&Image
   bIsVariable=true
   parent=Overlay_FuelMeter
   widget=FuelMeter
   properties {
    visibility=SelfHitTestInvisible
    brush {
     bIsDynamicallyLoaded=false
     drawAs=Image
     imageType=NoImage
     mirroring=NoMirror
     resourceName=None
     resourceObject=&/M_UI_Ferret_Speedometer_Inst_FuelNeedle
     tiling=NoTile
     imageSize=<185 185>
     margin=<ltrb 0 0 0 0>
     outlineSettings {
      bUseBrushTransparency=false
      roundingType=HalfHeightRadius
      width=0
      color {colorUseRule=UseColor_Specified, specifiedColor=<rgba 0 0 0 0>}
      cornerRadii=<0 0 0 0>
     }
     tintColor {colorUseRule=UseColor_Specified, specifiedColor=<rgba 1 1 1 1>}
     uVRegion {bIsValid=false, max=<0 0>, min=<0 0>}
    }
   }
   slot {class=OverlaySlot}
  }
  Image_MeterIcon {
   widgetClassPath=&Image
   bIsVariable=true
   parent=Overlay_FuelMeter
   widget=Image_MeterIcon
   properties {
    visibility=SelfHitTestInvisible
    brush {
     bIsDynamicallyLoaded=false
     drawAs=Image
     imageType=NoImage
     mirroring=NoMirror
     resourceName=None
     resourceObject=/Game/Athena/HUD/Materials/M_UI_FuelIcon_Fill
     tiling=NoTile
     imageSize=<20 20>
     margin=<ltrb 0 0 0 0>
     outlineSettings {
      bUseBrushTransparency=false
      roundingType=HalfHeightRadius
      width=0
      color {colorUseRule=UseColor_Specified, specifiedColor=<rgba 0 0 0 0>}
      cornerRadii=<0 0 0 0>
     }
     tintColor {colorUseRule=UseColor_Specified, specifiedColor=<rgba 1 1 1 1>}
     uVRegion {bIsValid=false, max=<0 0>, min=<0 0>}
    }
   }
   slot {class=OverlaySlot, properties {horizontalAlignment=HAlign_Center, verticalAlignment=VAlign_Bottom, padding=<ltrb 0 0 0 20>}}
  }
  Empty {
   widgetClassPath=&CommonTextBlock
   parent=Overlay_FuelMeter
   widget=Empty
   properties {
    text=E
    colorAndOpacity {colorUseRule=UseColor_Specified, specifiedColor=<rgba 0.9823 0.117 0.0003 1>}
    font {
     bForceMonospaced=false
     bMaterialIsStencil=false
     fontMaterial=null
     fontObject=&NotoSans
     letterSpacing=0
     monospacedWidth=1
     size=16
     skewAmount=0
     typefaceFontName=NotoSans-Bold
     outlineSettings {
      bApplyOutlineToDropShadows=false
      bMiteredCorners=false
      bSeparateFillAlpha=false
      outlineMaterial=null
      outlineSize=0
      outlineColor=<rgba 0 0 0 1>
     }
    }
    shadowOffset=<0 0>
   }
   slot {class=OverlaySlot, properties {horizontalAlignment=HAlign_Center, verticalAlignment=VAlign_Bottom, padding=<ltrb 0 0 33 28>}}
  }
  Full {
   widgetClassPath=&CommonTextBlock
   parent=Overlay_FuelMeter
   widget=Full
   properties {
    text=F
    font {
     bForceMonospaced=false
     bMaterialIsStencil=false
     fontMaterial=null
     fontObject=&NotoSans
     letterSpacing=0
     monospacedWidth=1
     size=16
     skewAmount=0
     typefaceFontName=NotoSans-Bold
     outlineSettings {
      bApplyOutlineToDropShadows=false
      bMiteredCorners=false
      bSeparateFillAlpha=false
      outlineMaterial=null
      outlineSize=0
      outlineColor=<rgba 0 0 0 1>
     }
    }
    shadowOffset=<0 0>
   }
   slot {class=OverlaySlot, properties {horizontalAlignment=HAlign_Center, verticalAlignment=VAlign_Bottom, padding=<ltrb 33 0 0 28>}}
  }
  Overlay_Values {
   widgetClassPath=&Overlay
   parent=Overlay_FuelMeter
   widget=Overlay_Values
   slot {class=OverlaySlot, properties {horizontalAlignment=HAlign_Center, verticalAlignment=VAlign_Center}}
  }
  Empty_1 {
   widgetClassPath=&CommonTextBlock
   parent=Overlay_Values
   widget=Empty_1
   properties {
    text=1/4
    font {
     bForceMonospaced=false
     bMaterialIsStencil=false
     fontMaterial=null
     fontObject=&Burbank
     letterSpacing=0
     monospacedWidth=1
     size=10
     skewAmount=0
     typefaceFontName=Burbank-PlayerPower
     outlineSettings {
      bApplyOutlineToDropShadows=false
      bMiteredCorners=false
      bSeparateFillAlpha=false
      outlineMaterial=null
      outlineSize=0
      outlineColor=<rgba 0 0 0 1>
     }
    }
    shadowOffset=<0 0>
   }
   slot {class=OverlaySlot, properties {horizontalAlignment=HAlign_Center, verticalAlignment=VAlign_Center, padding=<ltrb 0 0 45 13>}}
  }
  Empty_2 {
   widgetClassPath=&CommonTextBlock
   parent=Overlay_Values
   widget=Empty_2
   properties {
    text=1/2
    font {
     bForceMonospaced=false
     bMaterialIsStencil=false
     fontMaterial=null
     fontObject=&Burbank
     letterSpacing=0
     monospacedWidth=1
     size=10
     skewAmount=0
     typefaceFontName=Burbank-PlayerPower
     outlineSettings {
      bApplyOutlineToDropShadows=false
      bMiteredCorners=false
      bSeparateFillAlpha=false
      outlineMaterial=null
      outlineSize=0
      outlineColor=<rgba 0 0 0 1>
     }
    }
    shadowOffset=<0 0>
   }
   slot {class=OverlaySlot, properties {horizontalAlignment=HAlign_Center, verticalAlignment=VAlign_Center, padding=<ltrb 0 0 0 50>}}
  }
  Empty_3 {
   widgetClassPath=&CommonTextBlock
   parent=Overlay_Values
   widget=Empty_3
   properties {
    text=3/4
    font {
     bForceMonospaced=false
     bMaterialIsStencil=false
     fontMaterial=null
     fontObject=&Burbank
     letterSpacing=0
     monospacedWidth=1
     size=10
     skewAmount=0
     typefaceFontName=Burbank-PlayerPower
     outlineSettings {
      bApplyOutlineToDropShadows=false
      bMiteredCorners=false
      bSeparateFillAlpha=false
      outlineMaterial=null
      outlineSize=0
      outlineColor=<rgba 0 0 0 1>
     }
    }
    shadowOffset=<0 0>
   }
   slot {class=OverlaySlot, properties {horizontalAlignment=HAlign_Center, verticalAlignment=VAlign_Center, padding=<ltrb 45 0 0 13>}}
  }
  Overlay_0 {widgetClassPath=&Overlay, parent=VerticalBox_0, widget=Overlay_0, slot {class=VerticalBoxSlot}}
  BGGradientSpeed {
   widgetClassPath=&Image
   bIsVariable=true
   parent=Overlay_0
   widget=BGGradientSpeed
   properties {
    visibility=SelfHitTestInvisible
    brush {
     bIsDynamicallyLoaded=false
     drawAs=Image
     imageType=NoImage
     mirroring=NoMirror
     resourceName=None
     resourceObject=&M_UI_Ferret_Speedometer_Gradient
     tiling=NoTile
     imageSize=<165 165>
     margin=<ltrb 0 0 0 0>
     outlineSettings {
      bUseBrushTransparency=false
      roundingType=HalfHeightRadius
      width=0
      color {colorUseRule=UseColor_Specified, specifiedColor=<rgba 0 0 0 0>}
      cornerRadii=<0 0 0 0>
     }
     tintColor {colorUseRule=UseColor_Specified, specifiedColor=<rgba 0 0 0 1>}
     uVRegion {bIsValid=false, max=<0 0>, min=<0 0>}
    }
    colorAndOpacity=<rgba 0 0 0 0.6>
   }
   slot {class=OverlaySlot, properties {horizontalAlignment=HAlign_Center, verticalAlignment=VAlign_Center}}
  }
  Overlay_Boost {
   widgetClassPath=&Overlay
   parent=Overlay_0
   widget=Overlay_Boost
   slot {class=OverlaySlot, properties {horizontalAlignment=HAlign_Center}}
  }
  BoostGlowMeter {
   widgetClassPath=&Image
   bIsVariable=true
   parent=Overlay_Boost
   widget=BoostGlowMeter
   properties {
    visibility=SelfHitTestInvisible
    brush {
     bIsDynamicallyLoaded=false
     drawAs=Image
     imageType=NoImage
     mirroring=NoMirror
     resourceName=None
     resourceObject=&/M_UI_Ferret_Speedometer_Inst_Boost1Glow
     tiling=NoTile
     imageSize=<185 185>
     margin=<ltrb 0 0 0 0>
     outlineSettings {
      bUseBrushTransparency=false
      roundingType=HalfHeightRadius
      width=0
      color {colorUseRule=UseColor_Specified, specifiedColor=<rgba 0 0 0 0>}
      cornerRadii=<0 0 0 0>
     }
     tintColor {colorUseRule=UseColor_Specified, specifiedColor=<rgba 1 1 1 1>}
     uVRegion {bIsValid=false, max=<0 0>, min=<0 0>}
    }
   }
   slot {class=OverlaySlot, properties {horizontalAlignment=HAlign_Center, verticalAlignment=VAlign_Center}}
  }
  BoostMeter {
   widgetClassPath=&Image
   bIsVariable=true
   parent=Overlay_Boost
   widget=BoostMeter
   properties {
    visibility=SelfHitTestInvisible
    brush {
     bIsDynamicallyLoaded=false
     drawAs=Image
     imageType=NoImage
     mirroring=NoMirror
     resourceName=None
     resourceObject=&/M_UI_Ferret_Speedometer_Inst_Boost
     tiling=NoTile
     imageSize=<185 185>
     margin=<ltrb 0 0 0 0>
     outlineSettings {
      bUseBrushTransparency=false
      roundingType=HalfHeightRadius
      width=0
      color {colorUseRule=UseColor_Specified, specifiedColor=<rgba 0 0 0 0>}
      cornerRadii=<0 0 0 0>
     }
     tintColor {colorUseRule=UseColor_Specified, specifiedColor=<rgba 1 1 1 1>}
     uVRegion {bIsValid=false, max=<0 0>, min=<0 0>}
    }
   }
   slot {class=OverlaySlot, properties {horizontalAlignment=HAlign_Center, verticalAlignment=VAlign_Center}}
  }
  Image_BoostIcon {
   widgetClassPath=&Image
   bIsVariable=true
   parent=Overlay_Boost
   widget=Image_BoostIcon
   properties {
    visibility=SelfHitTestInvisible
    brush {
     bIsDynamicallyLoaded=false
     drawAs=Image
     imageType=NoImage
     mirroring=NoMirror
     resourceName=None
     resourceObject=/Game/Athena/HUD/Vehicles/Materials/M_UI_Boost
     tiling=NoTile
     imageSize=<120 120>
     margin=<ltrb 0 0 0 0>
     outlineSettings {
      bUseBrushTransparency=false
      roundingType=HalfHeightRadius
      width=0
      color {colorUseRule=UseColor_Specified, specifiedColor=<rgba 0 0 0 0>}
      cornerRadii=<0 0 0 0>
     }
     tintColor {colorUseRule=UseColor_Specified, specifiedColor=<rgba 1 1 1 1>}
     uVRegion {bIsValid=false, max=<0 0>, min=<0 0>}
    }
   }
   slot {class=OverlaySlot, properties {horizontalAlignment=HAlign_Center, verticalAlignment=VAlign_Bottom, padding=<ltrb 0 0 0 -30>}}
  }
  Overlay_SpeedMeter {widgetClassPath=&Overlay, parent=Overlay_0, widget=Overlay_SpeedMeter, slot {class=OverlaySlot}}
  SpeedMeter {
   widgetClassPath=&Image
   bIsVariable=true
   parent=Overlay_SpeedMeter
   widget=SpeedMeter
   properties {
    visibility=SelfHitTestInvisible
    brush {
     bIsDynamicallyLoaded=false
     drawAs=Image
     imageType=NoImage
     mirroring=NoMirror
     resourceName=None
     resourceObject=&/M_UI_Ferret_Speedometer_Inst_SpeedNeedle
     tiling=NoTile
     imageSize=<185 185>
     margin=<ltrb 0 0 0 0>
     outlineSettings {
      bUseBrushTransparency=false
      roundingType=HalfHeightRadius
      width=0
      color {colorUseRule=UseColor_Specified, specifiedColor=<rgba 0 0 0 0>}
      cornerRadii=<0 0 0 0>
     }
     tintColor {colorUseRule=UseColor_Specified, specifiedColor=<rgba 1 1 1 1>}
     uVRegion {bIsValid=false, max=<0 0>, min=<0 0>}
    }
   }
   slot {class=OverlaySlot}
  }
  ThinRingSpeed {
   widgetClassPath=&Image
   bIsVariable=true
   parent=Overlay_SpeedMeter
   widget=ThinRingSpeed
   properties {
    visibility=SelfHitTestInvisible
    brush {
     bIsDynamicallyLoaded=false
     drawAs=Image
     imageType=NoImage
     mirroring=NoMirror
     resourceName=None
     resourceObject=&M_UI_Ferret_Speedometer_Inst_ThinRing
     tiling=NoTile
     imageSize=<185 185>
     margin=<ltrb 0 0 0 0>
     outlineSettings {
      bUseBrushTransparency=false
      roundingType=HalfHeightRadius
      width=0
      color {colorUseRule=UseColor_Specified, specifiedColor=<rgba 0 0 0 0>}
      cornerRadii=<0 0 0 0>
     }
     tintColor {colorUseRule=UseColor_Specified, specifiedColor=<rgba 1 1 1 1>}
     uVRegion {bIsValid=false, max=<0 0>, min=<0 0>}
    }
   }
   slot {class=OverlaySlot}
  }
  SpeedGaugeImage {
   widgetClassPath=&Image
   bIsVariable=true
   parent=Overlay_SpeedMeter
   widget=SpeedGaugeImage
   properties {
    visibility=SelfHitTestInvisible
    brush {
     bIsDynamicallyLoaded=false
     drawAs=Image
     imageType=NoImage
     mirroring=NoMirror
     resourceName=None
     resourceObject=&/M_UI_Ferret_Speedometer_Inst_Speed
     tiling=NoTile
     imageSize=<185 185>
     margin=<ltrb 0 0 0 0>
     outlineSettings {
      bUseBrushTransparency=false
      roundingType=HalfHeightRadius
      width=0
      color {colorUseRule=UseColor_Specified, specifiedColor=<rgba 0 0 0 0>}
      cornerRadii=<0 0 0 0>
     }
     tintColor {colorUseRule=UseColor_Specified, specifiedColor=<rgba 1 1 1 1>}
     uVRegion {bIsValid=false, max=<0 0>, min=<0 0>}
    }
   }
   slot {class=OverlaySlot}
  }
  WBP_Vehicle_Warning {
   widgetClassPath=/Game/Athena/HUD/Vehicles/Dashboard/WBP_Vehicle_Warning
   bIsVariable=true
   parent=Overlay_Gauge
   widget=WBP_Vehicle_Warning
   properties {renderTransformPivot=<1 1>}
   slot {class=OverlaySlot, properties {horizontalAlignment=HAlign_Right, verticalAlignment=VAlign_Bottom, padding=<ltrb 0 0 275 355>}}
  }
  Overlay_LowFuel {
   widgetClassPath=&Overlay
   bIsVariable=true
   parent=Overlay_Gauge
   widget=Overlay_LowFuel
   properties {visibility=Collapsed, renderTransform {angle=0, scale=<0.75 0.75>, shear=<0 0>, translation=<0 0>}, renderTransformPivot=<1 1>}
   slot {class=OverlaySlot, properties {horizontalAlignment=HAlign_Right, verticalAlignment=VAlign_Bottom, padding=<ltrb 0 0 275 355>}}
  }
  Image_LowFuelIcon {
   widgetClassPath=&Image
   bIsVariable=true
   parent=Overlay_LowFuel
   widget=Image_LowFuelIcon
   properties {
    visibility=SelfHitTestInvisible
    brush {
     bIsDynamicallyLoaded=false
     drawAs=Image
     imageType=NoImage
     mirroring=NoMirror
     resourceName=None
     resourceObject=&M_UI_LowFuel_Inst
     tiling=NoTile
     imageSize=<200 200>
     margin=<ltrb 0 0 0 0>
     outlineSettings {
      bUseBrushTransparency=false
      roundingType=HalfHeightRadius
      width=0
      color {colorUseRule=UseColor_Specified, specifiedColor=<rgba 0 0 0 0>}
      cornerRadii=<0 0 0 0>
     }
     tintColor {colorUseRule=UseColor_Specified, specifiedColor=<rgba 1 1 1 1>}
     uVRegion {bIsValid=false, max=<0 0>, min=<0 0>}
    }
   }
   slot {class=OverlaySlot, properties {horizontalAlignment=HAlign_Center, verticalAlignment=VAlign_Center}}
  }
  Image_LowFuelPulse {
   widgetClassPath=&Image
   bIsVariable=true
   parent=Overlay_LowFuel
   widget=Image_LowFuelPulse
   properties {
    visibility=SelfHitTestInvisible
    brush {
     bIsDynamicallyLoaded=false
     drawAs=Image
     imageType=NoImage
     mirroring=NoMirror
     resourceName=None
     resourceObject=/Game/Athena/HUD/LTM/Wax/Materials/M_UI_TextGlow_Add_Gradient_Orange
     tiling=NoTile
     imageSize=<200 200>
     margin=<ltrb 0 0 0 0>
     outlineSettings {
      bUseBrushTransparency=false
      roundingType=HalfHeightRadius
      width=0
      color {colorUseRule=UseColor_Specified, specifiedColor=<rgba 0 0 0 0>}
      cornerRadii=<0 0 0 0>
     }
     tintColor {colorUseRule=UseColor_Specified, specifiedColor=<rgba 1 1 1 1>}
     uVRegion {bIsValid=false, max=<0 0>, min=<0 0>}
    }
   }
   slot {class=OverlaySlot, properties {horizontalAlignment=HAlign_Center, verticalAlignment=VAlign_Center}}
  }
  Image_RoundCard_Bolt {
   widgetClassPath=&Image
   bIsVariable=true
   parent=Overlay_LowFuel
   widget=Image_RoundCard_Bolt
   properties {
    visibility=SelfHitTestInvisible
    brush {
     bIsDynamicallyLoaded=false
     drawAs=Image
     imageType=NoImage
     mirroring=NoMirror
     resourceName=None
     resourceObject=/Game/UI/Frontend/Showdown/Materials/M_UI_SkewBoxBorder_Tournament_ModeAlertBorder_Tail
     tiling=NoTile
     imageSize=<16.953 89.0751>
     margin=<ltrb 0 0 0 0>
     outlineSettings {
      bUseBrushTransparency=false
      roundingType=HalfHeightRadius
      width=0
      color {colorUseRule=UseColor_Specified, specifiedColor=<rgba 0 0 0 0>}
      cornerRadii=<0 0 0 0>
     }
     tintColor {colorUseRule=UseColor_Specified, specifiedColor=<rgba 1 1 1 1>}
     uVRegion {bIsValid=false, max=<0 0>, min=<0 0>}
    }
    renderTransform {angle=0, scale=<1 1>, shear=<33 0>, translation=<81 77>}
    renderTransformPivot=<0.5 0>
   }
   slot {class=OverlaySlot, properties {horizontalAlignment=HAlign_Center, verticalAlignment=VAlign_Center}}
  }
  Switcher_FuelWarning {
   widgetClassPath=&CommonWidgetSwitcherLegacy
   bIsVariable=true
   parent=Overlay_LowFuel
   widget=Switcher_FuelWarning
   properties {transitionType=Horizontal, renderTransform {angle=0, scale=<1 1>, shear=<0 0>, translation=<142 80>}}
   slot {class=OverlaySlot}
  }
  CommonBorder_0 {
   widgetClassPath=&CommonBorder
   parent=Switcher_FuelWarning
   widget=CommonBorder_0
   properties {clipping=ClipToBounds, padding=<ltrb 0 0 0 0>, renderTransform {angle=0, scale=<1 1>, shear=<-12 0>, translation=<0 0>}}
   slot {class=WidgetSwitcherSlot, properties {horizontalAlignment=HAlign_Right, verticalAlignment=VAlign_Bottom}}
  }
  CommonTextBlock_57 {
   widgetClassPath=&CommonTextBlock
   parent=CommonBorder_0
   widget=CommonTextBlock_57
   properties {
    text=Low Fuel
    textTransformPolicy=ToUpper
    font {
     bForceMonospaced=false
     bMaterialIsStencil=false
     fontMaterial=null
     fontObject=&Burbank
     letterSpacing=0
     monospacedWidth=1
     size=24
     skewAmount=0
     typefaceFontName=BurbankRegular-Black
     outlineSettings {
      bApplyOutlineToDropShadows=false
      bMiteredCorners=false
      bSeparateFillAlpha=false
      outlineMaterial=null
      outlineSize=1
      outlineColor=<rgba 1 0.2021 0 1>
     }
    }
    shadowColorAndOpacity=<rgba 0.0625 0.0024 0 1>
    shadowOffset=<2 2>
   }
   slot {class=BorderSlot, properties {padding=<ltrb 0 0 0 0>}}
  }
  Overlay_NoFuelText {
   widgetClassPath=&Overlay
   bIsVariable=true
   parent=Switcher_FuelWarning
   widget=Overlay_NoFuelText
   properties {renderTransform {angle=0, scale=<1 1>, shear=<-12 0>, translation=<0 0>}}
   slot {class=WidgetSwitcherSlot, properties {horizontalAlignment=HAlign_Right, verticalAlignment=VAlign_Bottom}}
  }
  CommonTextBlock_1 {
   widgetClassPath=&CommonTextBlock
   parent=Overlay_NoFuelText
   widget=CommonTextBlock_1
   properties {
    text=No Fuel
    textTransformPolicy=ToUpper
    font {
     bForceMonospaced=false
     bMaterialIsStencil=false
     fontMaterial=null
     fontObject=&Burbank
     letterSpacing=0
     monospacedWidth=1
     size=24
     skewAmount=0
     typefaceFontName=BurbankRegular-Black
     outlineSettings {
      bApplyOutlineToDropShadows=false
      bMiteredCorners=false
      bSeparateFillAlpha=true
      outlineMaterial=null
      outlineSize=1
      outlineColor=<rgba 1 0.2021 0 1>
     }
    }
    shadowColorAndOpacity=<rgba 0.0625 0.0024 0 1>
    shadowOffset=<2 2>
   }
   slot {class=OverlaySlot, properties {horizontalAlignment=HAlign_Right}}
  }
  Overlay_NoFuelMessage {
   widgetClassPath=&Overlay
   bIsVariable=true
   parent=Overlay_160
   widget=Overlay_NoFuelMessage
   properties {visibility=Collapsed, renderTransform {angle=0, scale=<1 1>, shear=<-12 0>, translation=<-80 -150>}}
   slot {class=OverlaySlot, properties {horizontalAlignment=HAlign_Center, verticalAlignment=VAlign_Bottom}}
  }
  Image_darken {
   widgetClassPath=&Image
   bIsVariable=true
   parent=Overlay_NoFuelMessage
   widget=Image_darken
   properties {
    renderOpacity=0.5
    visibility=SelfHitTestInvisible
    brush {
     bIsDynamicallyLoaded=false
     drawAs=Image
     imageType=NoImage
     mirroring=NoMirror
     resourceName=None
     resourceObject=/Game/Athena/HUD/LTM/Wax/Materials/M_UI_TextGlow
     tiling=NoTile
     imageSize=<32 32>
     margin=<ltrb 0 0 0 0>
     outlineSettings {
      bUseBrushTransparency=false
      roundingType=HalfHeightRadius
      width=0
      color {colorUseRule=UseColor_Specified, specifiedColor=<rgba 0 0 0 0>}
      cornerRadii=<0 0 0 0>
     }
     tintColor {colorUseRule=UseColor_Specified, specifiedColor=<rgba 0 0 0 1>}
     uVRegion {bIsValid=false, max=<0 0>, min=<0 0>}
    }
   }
   slot {class=OverlaySlot, properties {horizontalAlignment=HAlign_Fill, verticalAlignment=VAlign_Fill, padding=<ltrb 0 -50 0 -50>}}
  }
  Image_messageBG {
   widgetClassPath=&Image
   bIsVariable=true
   parent=Overlay_NoFuelMessage
   widget=Image_messageBG
   properties {
    visibility=SelfHitTestInvisible
    brush {
     bIsDynamicallyLoaded=false
     drawAs=Image
     imageType=NoImage
     mirroring=NoMirror
     resourceName=None
     resourceObject=/Game/Athena/HUD/Vehicles/Materials/M_UI_FuelWarningMessageBG_Inst
     tiling=NoTile
     imageSize=<32 32>
     margin=<ltrb 0 0 0 0>
     outlineSettings {
      bUseBrushTransparency=false
      roundingType=HalfHeightRadius
      width=0
      color {colorUseRule=UseColor_Specified, specifiedColor=<rgba 0 0 0 0>}
      cornerRadii=<0 0 0 0>
     }
     tintColor {colorUseRule=UseColor_Specified, specifiedColor=<rgba 1 1 1 1>}
     uVRegion {bIsValid=false, max=<0 0>, min=<0 0>}
    }
   }
   slot {class=OverlaySlot, properties {horizontalAlignment=HAlign_Fill, verticalAlignment=VAlign_Fill}}
  }
  CommonTextBlock_15 {
   widgetClassPath=&CommonTextBlock
   parent=Overlay_NoFuelMessage
   widget=CommonTextBlock_15
   properties {
    justification=Center
    text="No Fuel!"
    textTransformPolicy=ToUpper
    font {
     bForceMonospaced=false
     bMaterialIsStencil=false
     fontMaterial=null
     fontObject=&Burbank
     letterSpacing=0
     monospacedWidth=1
     size=24
     skewAmount=0
     typefaceFontName=BurbankRegular-Black
     outlineSettings {
      bApplyOutlineToDropShadows=true
      bMiteredCorners=false
      bSeparateFillAlpha=true
      outlineMaterial=/Game/UI/Foundation/Materials/M-FontShadow
      outlineSize=3
      outlineColor=<rgba 0 0 0 0.8>
     }
    }
    shadowOffset=<0 0>
   }
   slot {class=OverlaySlot, properties {padding=<ltrb 200 0 200 0>}}
  }
  Widget_FuelGauge {widgetClassPath=/Script/FortniteUI.FortVehicleFuelGauge, bInherited=true, widget=null}
  HorizontalBox_FuelCluster {widgetClassPath=&HorizontalBox, bInherited=true, widget=null}
  Image_Speedometer {widgetClassPath=&Image, bInherited=true, widget=null}
  Image_BoostGlow {widgetClassPath=&Image, bInherited=true, widget=null}
  Image_BoostMeter {widgetClassPath=&Image, bInherited=true, widget=null}
  Overlay_BoostSparks {widgetClassPath=&Overlay, bInherited=true, widget=null}
  NumericText_Speed {widgetClassPath=/Script/CommonUI.CommonNumericTextBlock, bInherited=true, widget=null}
  Text_ItemName {widgetClassPath=&CommonTextBlock, bInherited=true, widget=null}
  TutorialHighlight {widgetClassPath=&TutorialHighlight}
 ]
}
