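# Translucent UI material for the biplane (Ferret) speedometer, a configurable radial gauge.
# Radial gradients and smoothsteps cut rings for background, border, fill and three tick tiers, clipped to an arc by angle gradients, with bookends.
# GaugeFill drives the fill arc, optional rotating needle texture and optional boost gradient, which blinks below a threshold.
# Switch scalars blend most layers, including low-fuel tint and full circle, through parallel color and alpha chains.
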Material /Game/Athena/DrivableVehicles/Biplane/UI/M_UI_Ferret_Speedometer
graph {
 properties {blendMode=BLEND_Translucent, materialDomain=MD_UI, numCustomizedUVs=2, pixelDepthOffsetMode=PDOM_Legacy, usedShadingModels=Default Lit}
 nodes [
  MaterialFunctionCall(Texture Scale=ScalarParameter() {defaultValue=0.9225, parameterName=CircleScale}) {materialFunction=&ScaleUVsByCenter}
  n1 = VectorParameter() {group=Bookend, parameterName=NeedleColor, defaultValue=<rgba 1 1 1 1>}
  n2 = VectorParameter() {desc="0.076, 0.848, 0, .5", group=Needle, parameterName=NeedleScale, defaultValue=<rgba 0.2 0.848 0 0.5>}
  n3 = AppendVector(A=@n2.B, B=@n2.B)
  n4 = TextureCoordinate()
  n5 = VectorParameter() {desc=".5, 1, 1, 0", group=Needle, parameterName=NeedleOffset, defaultValue=<rgba 0.5 1 1 0>}
  n6 = ScalarParameter() {defaultValue=0.25, desc=Detemines start location of gauge. 0 is 12 oclock. .25 is 9 oclock. -.25 is 3 oclock., parameterName=GaugeStart}
  n7 = ScalarParameter() {
   defaultValue=0.5
   desc=Determines the end location of the gauge relative to the start. 0 is the start location. Incrementing up to 1 will move the end clockwise.
   parameterName=GaugeLength
  }
  n8 = Subtract(A=@n6, B=@n7)
  n9 = ScalarParameter() {defaultValue=0.25, desc=.24, parameterName=GaugeFill}
  n10 = MaterialFunctionCall!(
   Texture Scale=@n2.RGB
   UVs=MaterialFunctionCall(
    UVs=Add(A=@n3, B=@n4)
    Rotation Center=AppendVector(A=@n5.R, B=@n5.R)
    "Rotation Angle (0-1)"=LinearInterpolate(A=@n6, B=@n8, Alpha=@n9) {constA=0.25, constB=0}
   ) {materialFunction=&CustomRotator}.Rotated Values
   Offset XY Texturesize Z=@n5.RGB
  ) {materialFunction=/Engine/Functions/Engine_MaterialFunctions01/Texturing/ScaleUVsAroundPoint}
  n11 = TextureSample(UVs=@n10.UVs) {texture=&/T_UI_FerretNeedle}
  n12 = ScalarParameter() {defaultValue=0.3, group=Needle, parameterName=NeedleEdge}
  n13 = Multiply(A=@n12, B=ScalarParameter() {defaultValue=0.3, group=Needle, parameterName=NeedleEdgeSoftness}) {constB=1.01}
  n14 = TextureCoordinate() {coordinateIndex=1}
  n15 = ScalarParameter() {defaultValue=0.18, desc=0.02, group=InnerBorder, parameterName=InnerBorderEdge}
  n16 = Multiply(A=@n15, B=ScalarParameter() {defaultValue=0.1, desc=0.1, parameterName=EdgeSoftness}) {constB=1.01}
  n17 = Subtract(
   A=MaterialFunctionCall(Alpha=Multiply(A=@n11.R, B=@n10.0-1 mask), Min=@n12, Max=Add(A=@n12, B=@n13)) {materialFunction=&SmoothStep}.Result
   B=MaterialFunctionCall(
    Alpha=MaterialFunctionCall(UVs=@n14) {materialFunction=&RadialGradientExponential}.RadialGradientExponential
    Min=@n15
    Max=Add(A=@n15, B=@n16)
   ) {materialFunction=&SmoothStep}.Result
  )
  n18 = ScalarParameter() {group=Needle, parameterName=IsNeedleOn, sortPriority=0}
  n19 = ScalarParameter() {defaultValue=0.25, desc=Detemines start location of gauge. 0 is 12 oclock. .25 is 9 oclock. -.25 is 3 oclock., parameterName=GaugeStart}
  n20 = ScalarParameter() {
   defaultValue=0.5
   desc=Determines the end location of the gauge relative to the start. 0 is the start location. Incrementing up to 1 will move the end clockwise.
   parameterName=GaugeLength
  }
  n21 = Constant()
  n22 = If(
   A=MaterialFunctionCall!(UVs, Rotation=@n19) {materialFunction=/Game/UI/Foundation/Materials/MF_UI_AngleGradient_DEPRECATED}.Result
   B=OneMinus(None=@n20)
   "A > B"=Constant() {r=1}
   "A == B"=@n21
   "A < B"=@n21
  )
  n23 = VectorParameter() {group=InnerBorder, parameterName=InnerBorderColor, defaultValue=<rgba 1 1 1 1>}
  n24 = ScalarParameter() {defaultValue=0.18, desc=0.02, group=InnerBorder, parameterName=InnerBorderEdge}
  n25 = Multiply(A=@n24, B=ScalarParameter() {defaultValue=0.1, desc=0.2, group=InnerBorder, parameterName=InnerBorderOutsideEdgeSoftness}) {constB=1.01}
  n26 = MaterialFunctionCall(UVs=TextureCoordinate() {coordinateIndex=1}) {materialFunction=&RadialGradientExponential}
  n27 = Add(B=ScalarParameter() {defaultValue=0.25, desc=0.135714, group=Gauge, parameterName=GaugeWidth})
  n28 = Multiply(A=@n27, B=ScalarParameter() {defaultValue=0.1, desc=0.2, group=InnerBorder, parameterName=InnerBorderEdgeSoftness}) {constB=1.01}
  n29 = Subtract(
   A=MaterialFunctionCall(Alpha=@n26.RadialGradientExponential, Min=@n24, Max=Add(A=@n24, B=@n25)) {materialFunction=&SmoothStep}.Result
   B=MaterialFunctionCall(Alpha=@n26.RadialGradientExponential, Min=@n27, Max=Add(A=@n27, B=@n28)) {materialFunction=&SmoothStep}.Result
  ) {constB=0}
  n30 = Multiply(A=@n23.A, B=@n29)
  n31 = VectorParameter() {group=Gauge, parameterName=GaugeBGColor, defaultValue=<rgba 1 1 1 0.25>}
  n32 = ScalarParameter() {defaultValue=0.02, desc=0.02, group=Gauge, parameterName=GaugeEdge}
  n33 = Multiply(A=@n32, B=ScalarParameter() {defaultValue=0.1, desc=.1, parameterName=OutsideEdgeSoftness}) {constB=1.01}
  n34 = MaterialFunctionCall(UVs=TextureCoordinate() {coordinateIndex=1}) {materialFunction=&RadialGradientExponential}
  n35 = Add(B=ScalarParameter() {defaultValue=0.25, desc=0.135714, group=Gauge, parameterName=GaugeWidth})
  n36 = Multiply(A=@n35, B=ScalarParameter() {defaultValue=0.1, desc=0.1, parameterName=EdgeSoftness}) {constB=1.01}
  n37 = Subtract(
   A=MaterialFunctionCall(Alpha=@n34.RadialGradientExponential, Min=@n32, Max=Add(A=@n32, B=@n33)) {materialFunction=&SmoothStep}.Result
   B=MaterialFunctionCall(Alpha=@n34.RadialGradientExponential, Min=@n35, Max=Add(A=@n35, B=@n36)) {materialFunction=&SmoothStep}.Result
  )
  n38 = ScalarParameter() {defaultValue=1, group=Gauge, parameterName=IsGaugeBGOn, sortPriority=0}
  n39 = LinearInterpolate(A=Constant(), B=Multiply(A=@n31.A, B=@n37), Alpha=@n38) {constAlpha=1, constB=0}
  n40 = VectorParameter() {group=TickMarks, parameterName=TickMarksColor, defaultValue=<rgba 1 1 1 1>}
  n41 = ScalarParameter() {defaultValue=0.05, desc=0.02, group=TickMarks, parameterName=TickMarkEdge}
  n42 = ScalarParameter() {defaultValue=0.1, desc=0.1, parameterName=EdgeSoftness}
  n43 = MaterialFunctionCall(UVs=TextureCoordinate() {coordinateIndex=1}) {materialFunction=&RadialGradientExponential}
  n44 = MaterialFunctionCall(Alpha=@n43.RadialGradientExponential, Min=@n41, Max=Add(A=@n41, B=Multiply(A=@n41, B=@n42) {constB=1.01})) {materialFunction=&SmoothStep}
  n45 = VectorParameter() {group=TickMarks, parameterName=TickMarkWidth, defaultValue=<rgba 0.3476 0.2143 0.1095 0>}
  n46 = Add(B=@n45.R)
  n47 = Multiply(A=@n46, B=ScalarParameter() {defaultValue=0.1, desc=0.1, parameterName=EdgeSoftness}) {constB=1.01}
  n48 = Subtract(
   A=@n44.Result
   B=MaterialFunctionCall(Alpha=@n43.RadialGradientExponential, Min=@n46, Max=Add(A=@n46, B=@n47)) {materialFunction=&SmoothStep}.Result
  ) {constB=0}
  n49 = VectorParameter() {group=TickMarks, parameterName=TickMarkEdgeSmooth, defaultValue=<rgba 0.01 0.01 0.01 0.01>}
  n50 = VectorParameter() {group=TickMarks, parameterName=TickMarkRepeat, defaultValue=<rgba 5 25 50 0>}
  n51 = VectorParameter() {group=TickMarks, parameterName=TickMarkRotation, defaultValue=<rgba 0.044 0.05 0.05 0>}
  n52 = VectorParameter() {group=TickMarks, parameterName=TickMarkSpacing, defaultValue=<rgba 0.9326 0.8286 0.781 0>}
  n53 = MaterialFunctionCall!(EdgeSmooth=@n49.R, Repeat=@n50.R, CenterOffset, Rotation=@n51.R, UV, SegmentSpacing=@n52.R) {materialFunction=/Game/Athena/HUD/reticles/MF_UI_RadialSegments_DEPRECATED}
  n54 = Add(B=@n45.G)
  n55 = Multiply(A=@n54, B=ScalarParameter() {defaultValue=0.1, desc=0.1, parameterName=EdgeSoftness}) {constB=1.01}
  n56 = Subtract(
   A=@n44.Result
   B=MaterialFunctionCall(Alpha=@n43.RadialGradientExponential, Min=@n54, Max=Add(A=@n54, B=@n55)) {materialFunction=&SmoothStep}.Result
  ) {constB=0}
  n57 = MaterialFunctionCall(EdgeSmooth=@n49.G, Repeat=@n50.G, Rotation=@n51.G, SegmentSpacing=@n52.G) {materialFunction=&MF_UI_RadialSegments_DEPRECATED}
  n58 = Add(B=@n45.B)
  n59 = ScalarParameter() {defaultValue=0.1, desc=0.1, parameterName=EdgeSoftness}
  n60 = MaterialFunctionCall(Alpha=@n43.RadialGradientExponential, Min=@n58, Max=Add(A=@n58, B=Multiply(A=@n58, B=@n59) {constB=1.01})) {materialFunction=&SmoothStep}
  n61 = Max(
   A=Max(A=Multiply(A=@n48, B=@n53.Result), B=Multiply(A=@n56, B=@n57.Result))
   B=Multiply(
    A=Subtract(A=@n44.Result, B=@n60.Result) {constB=0}
    B=MaterialFunctionCall(EdgeSmooth=@n49.B, Repeat=@n50.B, Rotation=@n51.B, SegmentSpacing=@n52.B) {materialFunction=&MF_UI_RadialSegments_DEPRECATED}.Result
   )
  )
  n62 = ScalarParameter() {defaultValue=1, group=TickMarks, parameterName=IsTickMarksOn, sortPriority=0}
  n63 = LinearInterpolate(A=@n39, B=LinearInterpolate(A=@n39, B=Multiply(A=@n40.A, B=@n61), Alpha=@n62), Alpha=@n61)
  n64 = ScalarParameter() {defaultValue=0.25, desc=Detemines start location of gauge. 0 is 12 oclock. .25 is 9 oclock. -.25 is 3 oclock., parameterName=GaugeStart}
  n65 = Constant() {r=1}
  n66 = ScalarParameter() {
   defaultValue=0.5
   desc=Determines the end location of the gauge relative to the start. 0 is the start location. Incrementing up to 1 will move the end clockwise.
   parameterName=GaugeLength
  }
  n67 = OneMinus(None=@n66)
  n68 = ScalarParameter() {defaultValue=0.25, desc=.24, parameterName=GaugeFill}
  n69 = Constant()
  n70 = ScalarParameter() {defaultValue=0.1, desc=0.02, group=Fill, parameterName=FillEdge}
  n71 = Add(A=@n70, B=Multiply(A=@n70, B=ScalarParameter() {defaultValue=0.1, desc=0.1, parameterName=EdgeSoftness}) {constB=1.01})
  n72 = MaterialFunctionCall(UVs=TextureCoordinate() {coordinateIndex=1}) {materialFunction=&RadialGradientExponential}
  n73 = Add(B=ScalarParameter() {defaultValue=0.2, desc=0.135714, group=Fill, parameterName=FillWidth})
  n74 = Add(A=@n73, B=Multiply(A=@n73, B=ScalarParameter() {defaultValue=0.1, desc=0.1, parameterName=EdgeSoftness}) {constB=1.01})
  n75 = Multiply(
   A=If(
    A=MaterialFunctionCall(Rotation=@n64) {materialFunction=&MF_UI_AngleGradient_DEPRECATED}.Result
    B=LinearInterpolate(A=@n65, B=@n67, Alpha=@n68)
    "A > B"=Constant() {r=1}
    "A == B"=@n69
    "A < B"=@n69
   )
   B=Subtract(
    A=MaterialFunctionCall(Alpha=@n72.RadialGradientExponential, Min=@n70, Max=@n71) {materialFunction=&SmoothStep}.Result
    B=MaterialFunctionCall(Alpha=@n72.RadialGradientExponential, Min=@n73, Max=@n74) {materialFunction=&SmoothStep}.Result
   )
  )
  n76 = LinearInterpolate(
   A=@n63
   B=LinearInterpolate(A=@n63, B=@n75, Alpha=ScalarParameter() {defaultValue=1, group=InnerBorder, parameterName=GaugeFillOn, sortPriority=0})
   Alpha=@n75
  )
  n77 = VectorParameter() {group=Bookend, parameterName=BookendsColor, defaultValue=<rgba 1 1 1 1>}
  n78 = ScalarParameter() {defaultValue=0.25, desc=Detemines start location of gauge. 0 is 12 oclock. .25 is 9 oclock. -.25 is 3 oclock., parameterName=GaugeStart}
  n79 = ScalarParameter() {defaultValue=0.005, group=Bookend, parameterName=BookendSize}
  n80 = OneMinus(None=@n79)
  n81 = Constant()
  n82 = ScalarParameter() {
   defaultValue=0.5
   desc=Determines the end location of the gauge relative to the start. 0 is the start location. Incrementing up to 1 will move the end clockwise.
   parameterName=GaugeLength
  }
  n83 = Add(A=Subtract(A=@n78, B=@n82), B=@n79)
  n84 = Constant()
  n85 = Add(
   A=If(
    A=MaterialFunctionCall(Rotation=@n78) {materialFunction=&MF_UI_AngleGradient_DEPRECATED}.Result
    B=@n80
    "A > B"=Constant() {r=1}
    "A == B"=@n81
    "A < B"=@n81
   )
   B=If(
    A=MaterialFunctionCall(Rotation=@n83) {materialFunction=&MF_UI_AngleGradient_DEPRECATED}.Result
    B=@n80
    "A > B"=Constant() {r=1}
    "A == B"=@n84
    "A < B"=@n84
   )
  )
  n86 = ScalarParameter() {defaultValue=0.04, desc=0.02, group=Bookend, parameterName=BookendEdge}
  n87 = Multiply(A=@n86, B=ScalarParameter() {defaultValue=0.1, desc=0.1, parameterName=EdgeSoftness}) {constB=1.01}
  n88 = MaterialFunctionCall(UVs=TextureCoordinate() {coordinateIndex=1}) {materialFunction=&RadialGradientExponential}
  n89 = Add(B=ScalarParameter() {defaultValue=0.25, desc=0.135714, group=Gauge, parameterName=GaugeWidth})
  n90 = Multiply(A=@n89, B=ScalarParameter() {defaultValue=0.1, desc=0.1, parameterName=EdgeSoftness}) {constB=1.01}
  n91 = Subtract(
   A=MaterialFunctionCall(Alpha=@n88.RadialGradientExponential, Min=@n86, Max=Add(A=@n86, B=@n87)) {materialFunction=&SmoothStep}.Result
   B=MaterialFunctionCall(Alpha=@n88.RadialGradientExponential, Min=@n89, Max=Add(A=@n89, B=@n90)) {materialFunction=&SmoothStep}.Result
  )
  n92 = Multiply(A=@n91, B=Multiply(A=@n77.A, B=@n85))
  n93 = ScalarParameter() {defaultValue=1, group=Bookend, parameterName=IsBookendsOn, sortPriority=0}
  n94 = LinearInterpolate(A=Constant(), B=@n92, Alpha=@n93)
  n95 = LinearInterpolate(A=Multiply(A=@n22, B=Max(A=@n76, B=@n30)), B=@n94, Alpha=@n94)
  n96 = Max(A=LinearInterpolate(A=Constant(), B=Multiply(A=@n1.A, B=@n17), Alpha=@n18), B=@n95)
  n97 = VectorParameter() {group=FullCircle, parameterName=FullCircleColor, defaultValue=<rgba 1 1 1 1>}
  n98 = ScalarParameter() {defaultValue=0.18, desc=0.02, group=FullCircle, parameterName=FullCircleEdge}
  n99 = Multiply(A=@n98, B=ScalarParameter() {defaultValue=0.1, desc=0.2, group=FullCircle, parameterName=FullCircleEdgeSoftness}) {constB=1.01}
  n100 = MaterialFunctionCall(UVs=TextureCoordinate() {coordinateIndex=1}) {materialFunction=&RadialGradientExponential}
  n101 = Add(B=ScalarParameter() {defaultValue=0.25, desc=0.135714, group=FullCircle, parameterName=FullCircleWidth})
  n102 = Multiply(A=@n101, B=ScalarParameter() {defaultValue=0.1, desc=0.2, group=FullCircle, parameterName=FullCircleEdgeSoftness}) {constB=1.01}
  n103 = Subtract(
   A=MaterialFunctionCall(Alpha=@n100.RadialGradientExponential, Min=@n98, Max=Add(A=@n98, B=@n99)) {materialFunction=&SmoothStep}.Result
   B=MaterialFunctionCall(Alpha=@n100.RadialGradientExponential, Min=@n101, Max=Add(A=@n101, B=@n102)) {materialFunction=&SmoothStep}.Result
  ) {constB=0}
  n104 = ScalarParameter() {group=FullCircle, parameterName=IsFullCircleOn, sortPriority=0}
  Max(A=@n96, B=LinearInterpolate(A=@n96, B=Multiply(A=@n97.A, B=@n103), Alpha=@n104) {constAlpha=1, constB=0})
  n105 = LinearInterpolate(
   A=Constant()
   B=Multiply(A=Multiply(A=@n29, B=@n23.RGB), B=@n23.A)
   Alpha=ScalarParameter() {defaultValue=1, group=InnerBorder, parameterName=IsInnerBorderOn, sortPriority=0}
  ) {constAlpha=1, constB=0}
  n106 = LinearInterpolate(A=Constant(), B=Multiply(A=Multiply(A=@n37, B=@n31.RGB), B=@n31.A), Alpha=@n38) {constAlpha=1, constB=0}
  n107 = Multiply(A=@n61, B=@n40.RGB)
  n108 = LinearInterpolate(A=@n106, B=LinearInterpolate(A=@n106, B=Multiply(A=@n107, B=@n40.A), Alpha=@n62) {constAlpha=1, constB=0}, Alpha=@n61)
  n109 = ScalarParameter() {defaultValue=0.25, parameterName=GaugeFill}
  n110 = LinearInterpolate(
   A=VectorParameter() {group=BoostGradient, parameterName=ChargingColor1, defaultValue=<rgba 0.7913 0 0 1>}.RGB
   B=VectorParameter() {group=BoostGradient, parameterName=ChargingColor2, defaultValue=<rgba 1 0.4508 0 1>}.RGB
   Alpha=Power(Base=@n109, Exp=ScalarParameter() {defaultValue=3, group=BoostGradient, parameterName=lerpPow})
  )
  n111 = VectorParameter() {group=BoostGradient, parameterName=ReadyColor, defaultValue=<rgba 0.0284 1 0 1>}
  n112 = Floor(None=@n109)
  n113 = ScalarParameter() {defaultValue=0.5, group=BoostGradient, parameterName=BlinkValue}
  n114 = Constant() {r=1}
  n115 = MaterialFunctionCall!(
   Sine Phase=Multiply(A=ScalarParameter() {defaultValue=5, group=BoostGradient, parameterName=BlinkSpeed}, B=Time())
   Value 1=ScalarParameter() {defaultValue=1.5, group=BoostGradient, parameterName=s}
   Value 2=ScalarParameter() {defaultValue=0.5, group=BoostGradient, parameterName=s_1}
  ) {materialFunction=/Engine/Functions/Engine_MaterialFunctions02/Utility/Sine_Remapped}
  n116 = LinearInterpolate(
   A=VectorParameter() {group=Fill, parameterName=FillColor, defaultValue=<rgba 1 1 1 1>}.RGB
   B=Multiply(A=LinearInterpolate(A=@n110, B=@n111.RGB, Alpha=@n112), B=If(A=@n109, B=@n113, "A > B"=@n114, "A == B"=@n114, "A < B"=@n115.Remapped Sine))
   Alpha=ScalarParameter() {group=BoostGradient, parameterName=UseBoostGradient, sortPriority=0}
  )
  n117 = LinearInterpolate(
   A=@n108
   B=LinearInterpolate(A=@n108, B=Multiply(A=@n75, B=@n116), Alpha=ScalarParameter() {defaultValue=1, group=InnerBorder, parameterName=GaugeFillOn, sortPriority=0}) {constAlpha=1, constB=0}
   Alpha=@n75
  )
  n118 = Constant()
  n119 = Multiply(A=Multiply(A=@n85, B=@n77.RGB), B=@n77.A)
  n120 = LinearInterpolate(
   A=Multiply(A=Max(A=@n117, B=@n105), B=@n22)
   B=Multiply(A=@n91, B=LinearInterpolate(A=@n118, B=@n119, Alpha=@n93) {constAlpha=1, constB=0})
   Alpha=@n92
  )
  n121 = ScalarParameter() {defaultValue=0.25, desc=Detemines start location of gauge. 0 is 12 oclock. .25 is 9 oclock. -.25 is 3 oclock., parameterName=GaugeStart}
  n122 = Constant() {r=1}
  n123 = ScalarParameter() {
   defaultValue=0.5
   desc=Determines the end location of the gauge relative to the start. 0 is the start location. Incrementing up to 1 will move the end clockwise.
   parameterName=GaugeLength
  }
  n124 = OneMinus(None=@n123)
  n125 = ScalarParameter() {defaultValue=0.1, desc=.24, parameterName=LowFuelLength}
  n126 = Constant()
  n127 = Multiply(
   A=If(
    A=MaterialFunctionCall(Rotation=@n121) {materialFunction=&MF_UI_AngleGradient_DEPRECATED}.Result
    B=LinearInterpolate(A=@n122, B=@n124, Alpha=@n125)
    "A > B"=Constant() {r=1}
    "A == B"=@n126
    "A < B"=@n126
   )
   B=@n95
  )
  n128 = Multiply(A=@n127, B=VectorParameter() {group=Fill, parameterName=LowFuelColor, defaultValue=<rgba 0.9823 0.117 0.0003 1>}.RGB)
  n129 = Multiply(A=@n17, B=@n1.RGB)
  n130 = Max(
   A=LinearInterpolate(
    A=@n120
    B=LinearInterpolate(A=@n120, B=@n128, Alpha=@n127)
    Alpha=ScalarParameter() {group=Fill, parameterName=IsLowFuelFillOn, sortPriority=0}
   )
   B=LinearInterpolate(A=Constant(), B=Multiply(A=@n129, B=@n1.A), Alpha=@n18) {constAlpha=1, constB=0}
  )
  n131 = Multiply(A=@n103, B=@n97.RGB)
  Max(A=@n130, B=LinearInterpolate(A=@n130, B=Multiply(A=@n131, B=@n97.A), Alpha=@n104) {constAlpha=1, constB=0})
  ScalarParameter() {defaultValue=0.25, desc=0.135714, group=InnerBorder, parameterName=InnerBorderWidth}
  ScalarParameter() {defaultValue=0.25, desc=0.135714, group=InnerBorder, parameterName=InnerBorderWidth}
 ]
}
